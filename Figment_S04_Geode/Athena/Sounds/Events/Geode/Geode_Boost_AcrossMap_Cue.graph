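# Sound cue for the Geode boost across the map, a loud, high-priority thrust sound in an important in-game sound class with a very large max distance.
# A quality-level switch gives the first two levels a thrust wave through close and distant attenuations, mixed with a random meteor pass-by on overridden pass-by attenuation.
# The last two levels use only a looping "opt" thrust wave through close and distant-rocket attenuations, dropping the pass-by layer.
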SoundCue /Figment_S04_Geode/Athena/Sounds/Events/Geode/Geode_Boost_AcrossMap_Cue
root=QualityLevel_0
nodes [
 {id=Attenuation_0, children [WavePlayer_2], properties {attenuationSettings=&/Geode_Close_Att}}
 {id=Attenuation_1, children [WavePlayer_2], properties {attenuationSettings=&/Geode_Distant_Att}}
 {
  id=Attenuation_2
  children [Random_1]
  properties {
   attenuationSettings=&/Geode_PassBy_Att
   attenuationOverrides {falloffDistance=2000, stereoSpread=1000, attenuationShapeExtents {x=3000}}
  }
 }
 {id=Attenuation_3, children [WavePlayer_1], properties {attenuationSettings=&Geode_Close_Att}}
 {
  id=Attenuation_4
  children [WavePlayer_1]
  properties {attenuationSettings=&/Geode_Distant_Rocket_Att}
 }
 {id=Mixer_0, children [Attenuation_0, Attenuation_1, Attenuation_2], properties {inputVolume [1,1,1]}}
 {id=Mixer_1, children [Attenuation_3, Attenuation_4], properties {inputVolume [1,1]}}
 {id=QualityLevel_0, children [Mixer_0,Mixer_0,Mixer_1,Mixer_1]}
 {id=Random_1, children [WavePlayer_7,WavePlayer_8], properties {weights [1,1]}}
 {id=WavePlayer_1, properties {bLooping=true, soundWaveAssetPtr=&/Geode_Tear_Thrust_14_opt}}
 {id=WavePlayer_2, properties {soundWaveAssetPtr=&/Geode_Tear_Thrust_14}}
 {id=WavePlayer_7, properties {soundWaveAssetPtr=/Game/Sounds/Fort_GamePlay_Sounds/Meteor/Meteor_Almost_Landed_01}}
 {id=WavePlayer_8, properties {soundWaveAssetPtr=/Game/Sounds/Fort_GamePlay_Sounds/Meteor/Meteor_Almost_Landed_02}}
]
properties {
 maxDistance=2097152
 priority=100
 soundClassObject=&Fort_Mission_Ingame_Important_PSM
 volumeMultiplier=0.55
}
