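# Sound cue for the Geode boost pre-jettison sound, set as high priority with a very long audible range.
# A quality-level root sends the first two levels to a full mix and the last two to a lighter mix.
# Both mixes layer one looping thrust wave through close and distant-rocket attenuations, the lighter using an optimized wave.
# The full mix adds a randomly chosen, non-looping meteor sound through an overridden pass-by attenuation.
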SoundCue /Figment_S04_Geode/Athena/Sounds/Events/Geode/Geode_Boost_PreJettison_Cue
root=QualityLevel_0
nodes [
 {id=Attenuation_0, children [WavePlayer_1], properties {attenuationSettings=&/Geode_Close_Att}}
 {
  id=Attenuation_1
  children [WavePlayer_1]
  properties {attenuationSettings=&/Geode_Distant_Rocket_Att}
 }
 {id=Attenuation_3, children [WavePlayer_0], properties {attenuationSettings=&Geode_Close_Att}}
 {
  id=Attenuation_4
  children [WavePlayer_0]
  properties {attenuationSettings=&Geode_Distant_Rocket_Att}
 }
 {
  id=Attenuation_5
  children [Random_0]
  properties {
   attenuationSettings=&/Geode_PassBy_Att
   attenuationOverrides {falloffDistance=2000, stereoSpread=1000, attenuationShapeExtents {x=3000}}
  }
 }
 {id=Mixer_0, children [Attenuation_0, Attenuation_1], properties {inputVolume [1,1]}}
 {id=Mixer_1, children [Attenuation_3, Attenuation_4, Attenuation_5], properties {inputVolume [1,1,1]}}
 {id=QualityLevel_0, children [Mixer_1,Mixer_1,Mixer_0,Mixer_0]}
 {id=Random_0, children [WavePlayer_4,WavePlayer_5], properties {weights [1,1]}}
 {id=WavePlayer_0, properties {bLooping=true, soundWaveAssetPtr=&/Geode_Tear_Thrust_07}}
 {id=WavePlayer_1, properties {bLooping=true, soundWaveAssetPtr=&/Geode_Tear_Thrust_07_opt}}
 {id=WavePlayer_4, properties {soundWaveAssetPtr=/Game/Sounds/Fort_GamePlay_Sounds/Meteor/Meteor_Almost_Landed_01}}
 {id=WavePlayer_5, properties {soundWaveAssetPtr=/Game/Sounds/Fort_GamePlay_Sounds/Meteor/Meteor_Almost_Landed_02}}
]
properties {
 attenuationSettings=&Geode_Distant_Rocket_Att
 maxDistance=2097152
 priority=100
 soundClassObject=&Fort_Mission_Ingame_Important_PSM
 volumeMultiplier=0.55
}
