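# Sound cue for the wind freefall sound of the Geode event, played as a high-priority important mission sound.
# A mixer layers two attenuation branches, close and distant, that share one quality-level node.
# That node picks between the standard wind wave and its Opt variant, each filling two quality slots.
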SoundCue /Figment_S04_Geode/Athena/Sounds/Events/Geode/Geode_Freefall_Cue
root=Mixer_1
nodes [
 {id=Attenuation_2, children [QualityLevel_1], properties {attenuationSettings=&/Geode_Close_Att}}
 {
  id=Attenuation_3
  children [QualityLevel_1]
  properties {attenuationSettings=&/Geode_Distant_Att}
 }
 {id=Mixer_1, children [Attenuation_2, Attenuation_3], properties {inputVolume [1,1]}}
 {id=QualityLevel_1, children [WavePlayer_1,WavePlayer_1,WavePlayer_2,WavePlayer_2]}
 {id=WavePlayer_1, properties {soundWaveAssetPtr=&/Geode_Tear_WindFreefall_Figment_12}}
 {id=WavePlayer_2, properties {soundWaveAssetPtr=&/Geode_Tear_WindFreefall_Figment_12_Opt}}
]
properties {maxDistance=2097152, priority=100, soundClassObject=&Fort_Mission_Ingame_Important_PSM}
