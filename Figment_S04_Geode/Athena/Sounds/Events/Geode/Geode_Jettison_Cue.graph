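# Sound cue for the Geode jettison event, built on a single Geode tear thrust-activation recording.
# A mixer layers two branches of that same sound, one with close attenuation and one with distant attenuation.
# Both branches share a quality-level switch that uses the full wave for the first two tiers and an optimized variant for the others.
# It plays at reduced volume with high priority, a very large max distance and an important in-game mission sound class.
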SoundCue /Figment_S04_Geode/Athena/Sounds/Events/Geode/Geode_Jettison_Cue
root=Mixer_0
nodes [
 {id=Attenuation_0, children [QualityLevel_0], properties {attenuationSettings=&/Geode_Close_Att}}
 {
  id=Attenuation_1
  children [QualityLevel_0]
  properties {attenuationSettings=&/Geode_Distant_Att}
 }
 {id=Mixer_0, children [Attenuation_0, Attenuation_1], properties {inputVolume [1,1]}}
 {id=QualityLevel_0, children [WavePlayer_3,WavePlayer_3,WavePlayer_0,WavePlayer_0]}
 {id=WavePlayer_0, properties {soundWaveAssetPtr=&/Geode_Tear_ThrustActivate_08_opt}}
 {id=WavePlayer_3, properties {soundWaveAssetPtr=&/Geode_Tear_ThrustActivate_08}}
]
properties {
 maxDistance=2097152
 priority=100
 soundClassObject=&Fort_Mission_Ingame_Important_PSM
 volumeMultiplier=0.55
}
