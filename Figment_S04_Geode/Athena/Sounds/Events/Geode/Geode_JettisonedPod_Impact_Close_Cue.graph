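# Sound cue for the close-range impact of a jettisoned pod in the Geode event.
# It plays a single reused meteor impact wave through an attenuation node using the Geode close attenuation settings.
# The cue is high priority, routed to an important in-game mission sound class, with reduced volume and a long maximum audible distance.
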SoundCue /Figment_S04_Geode/Athena/Sounds/Events/Geode/Geode_JettisonedPod_Impact_Close_Cue
root=Attenuation_0
nodes [
 {id=Attenuation_0, children [WavePlayer_0], properties {attenuationSettings=&/Geode_Close_Att}}
 {id=WavePlayer_0, properties {soundWaveAssetPtr=/Figment_S04_Geode/Sounds/Fort_GamePlay_Sounds/Meteor/Meteor_Impact_01}}
]
properties {
 maxDistance=20000
 priority=100
 soundClassObject=&Fort_Mission_Ingame_Important_PSM
 volumeMultiplier=0.55
}
