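# Sound cue for the distant impact of a jettisoned pod in the Geode event.
# It plays a single distant meteor-landing wave through an attenuation node that uses the shared Geode distant attenuation settings.
# The cue has top priority, a very large max distance and reduced volume, and routes to an important in-game mission sound class.
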SoundCue /Figment_S04_Geode/Athena/Sounds/Events/Geode/Geode_JettisonedPod_Impact_Distant_Cue
root=Attenuation_1
nodes [
 {id=Attenuation_1, children [WavePlayer_1], properties {attenuationSettings=&/Geode_Distant_Att}}
 {id=WavePlayer_1, properties {soundWaveAssetPtr=/Figment_S04_Geode/Sounds/Fort_GamePlay_Sounds/Meteor/Meteor_Land_Distant}}
]
properties {
 maxDistance=2097152
 priority=100
 soundClassObject=&Fort_Mission_Ingame_Important_PSM
 volumeMultiplier=0.55
}
