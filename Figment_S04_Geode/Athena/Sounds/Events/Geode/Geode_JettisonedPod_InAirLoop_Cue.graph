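# Sound cue for the in-air sound of a jettisoned pod in Geode, built from meteor falling sounds.
# A mixer plays two layers together: a looping distant-falling wave through a close attenuation, and a non-looping almost-landed wave through a pass-by attenuation.
# Each layer gets its own spatial falloff, and the cue uses an important in-game mission sound class with raised priority and reduced volume.
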SoundCue /Figment_S04_Geode/Athena/Sounds/Events/Geode/Geode_JettisonedPod_InAirLoop_Cue
root=Mixer_0
nodes [
 {id=Attenuation_0, children [WavePlayer_1], properties {attenuationSettings=&/Geode_Close_Att}}
 {id=Attenuation_2, children [WavePlayer_2], properties {attenuationSettings=&/Geode_PassBy_Att}}
 {id=Mixer_0, children [Attenuation_0, Attenuation_2], properties {inputVolume [1,1]}}
 {
  id=WavePlayer_1
  properties {bLooping=true, soundWaveAssetPtr=/Figment_S04_Geode/Sounds/Fort_GamePlay_Sounds/Meteor/Meteor_Falling_Distant_01}
 }
 {id=WavePlayer_2, properties {soundWaveAssetPtr=/Game/Sounds/Fort_GamePlay_Sounds/Meteor/Meteor_Almost_Landed_01}}
]
properties {
 maxDistance=20000
 priority=100
 soundClassObject=&Fort_Mission_Ingame_Important_PSM
 volumeMultiplier=0.55
}
