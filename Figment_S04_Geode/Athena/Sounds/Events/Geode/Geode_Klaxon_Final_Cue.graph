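# Sound cue for the final klaxon in the Geode event, played as a high-priority in-game SFX with a very large max distance.
# Each mix layers a close and a distant klaxon wave, each with its own attenuation, with the close layer slightly boosted.
# A quality-level switch sends the first two levels to the standard waves and the last two to the "_opt" variants.
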SoundCue /Figment_S04_Geode/Athena/Sounds/Events/Geode/Geode_Klaxon_Final_Cue
root=QualityLevel_0
nodes [
 {
  id=Attenuation_4
  children [WavePlayer_2]
  properties {attenuationSettings=&/Geode_Klaxxon_Close_Att}
 }
 {
  id=Attenuation_5
  children [WavePlayer_1]
  properties {attenuationSettings=&/Geode_Distant_Klaxxon_Att}
 }
 {
  id=Attenuation_6
  children [WavePlayer_5]
  properties {attenuationSettings=&Geode_Klaxxon_Close_Att}
 }
 {
  id=Attenuation_7
  children [WavePlayer_6]
  properties {attenuationSettings=&Geode_Distant_Klaxxon_Att}
 }
 {id=Mixer_0, children [Attenuation_4, Attenuation_5], properties {inputVolume [1.2,1]}}
 {id=Mixer_3, children [Attenuation_6, Attenuation_7], properties {inputVolume [1.2,1]}}
 {id=QualityLevel_0, children [Mixer_0,Mixer_0,Mixer_3,Mixer_3]}
 {id=WavePlayer_1, properties {soundWaveAssetPtr=&/Geode_Tear_Klaxxon_Distant_19}}
 {id=WavePlayer_2, properties {soundWaveAssetPtr=&/Geode_Tear_Klaxxon_Close_18}}
 {id=WavePlayer_5, properties {soundWaveAssetPtr=&/Geode_Tear_Klaxxon_Close_18_opt}}
 {id=WavePlayer_6, properties {soundWaveAssetPtr=&/Geode_Tear_Klaxxon_Distant_19_opt}}
]
properties {maxDistance=2097152, priority=100, soundClassObject=&Fort_SFX_InGame, volumeMultiplier=1}
