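# Looping Sound Cue for the Geode steam buildup, attenuated by the Geode interior settings.
# It layers a smoke-grenade release loop with a quieter Geode tear pre-takeoff idle loop, using endlessly looping envelopers with opposed volume curves so the layers swell alternately.
# A quality node routes the first two levels to the standard-wave mix and the last two to a matching mix of "_opt" waves, whose idle layer is a different numbered take.
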SoundCue /Figment_S04_Geode/Athena/Sounds/Events/Geode/Geode_SteamBuildup_Loop_Cue
root=QualityLevel_0
nodes [
 {
  id=Enveloper_0
  children [WavePlayer_0]
  properties {
   bLoop=true
   bLoopIndefinitely=true
   loopEnd=40
   pitchCurve {editorCurveData {keys [{time=0}]}}
   volumeCurve {editorCurveData {keys [<at 0 1>,<at 20 0.5>,<at 40 1>]}}
  }
 }
 {
  id=Enveloper_1
  children [WavePlayer_1]
  properties {
   bLoop=true
   bLoopIndefinitely=true
   loopEnd=40
   pitchCurve {editorCurveData {keys [{time=0}]}}
   volumeCurve {editorCurveData {keys [<at 0 0.5>,<at 20 1>,<at 40 0.5>]}}
  }
 }
 {
  id=Enveloper_2
  children [WavePlayer_4]
  properties {
   bLoop=true
   bLoopIndefinitely=true
   loopEnd=40
   pitchCurve {editorCurveData {keys [{time=0}]}}
   volumeCurve {editorCurveData {keys [<at 0 1>,<at 20 0.5>,<at 40 1>]}}
  }
 }
 {
  id=Enveloper_3
  children [WavePlayer_3]
  properties {
   bLoop=true
   bLoopIndefinitely=true
   loopEnd=40
   pitchCurve {editorCurveData {keys [{time=0}]}}
   volumeCurve {editorCurveData {keys [<at 0 0.5>,<at 20 1>,<at 40 0.5>]}}
  }
 }
 {id=Mixer_0, children [Enveloper_0,Enveloper_1], properties {inputVolume [1,0.5]}}
 {id=Mixer_2, children [Enveloper_2,Enveloper_3], properties {inputVolume [1,0.5]}}
 {id=QualityLevel_0, children [Mixer_0,Mixer_0,Mixer_2,Mixer_2]}
 {id=WavePlayer_0, properties {bLooping=true, soundWaveAssetPtr=/Game/Athena/Sounds/Weapons/Smoke_Grenade/sfx_SmokeGrenade_ReleaseLoop_01}}
 {id=WavePlayer_1, properties {bLooping=true, soundWaveAssetPtr=&/Geode_Tear_PreTakeoffIdle_13}}
 {
  id=WavePlayer_3
  properties {bLooping=true, soundWaveAssetPtr=&/Geode_Tear_PreTakeoffIdle_15_opt}
 }
 {id=WavePlayer_4, properties {bLooping=true, soundWaveAssetPtr=/Game/Athena/Sounds/Weapons/Smoke_Grenade/sfx_SmokeGrenade_ReleaseLoop_01_opt}}
]
properties {
 attenuationSettings=&/Geode_Inside_Att
 maxDistance=10000
 priority=100
 soundClassObject=&Fort_Mission_Ingame_Important_PSM
 volumeMultiplier=0.3
}
