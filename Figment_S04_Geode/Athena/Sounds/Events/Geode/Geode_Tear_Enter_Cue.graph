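# Sound cue for the Geode tear-enter event, playing one of two randomly chosen wave variants on an important in-game mission sound class.
# The random pick feeds both a close and a distant attenuation node, which a mixer layers at equal volume to shape the sound across distance.
# A quality-level root routes its first two slots to that mixer and leaves the other two empty.
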SoundCue /Figment_S04_Geode/Athena/Sounds/Events/Geode/Geode_Tear_Enter_Cue
root=QualityLevel_0
nodes [
 {id=Attenuation_0, children [Random_1], properties {attenuationSettings=&/Geode_Close_Att}}
 {id=Attenuation_1, children [Random_1], properties {attenuationSettings=&/Geode_Distant_Att}}
 {id=Mixer_0, children [Attenuation_0, Attenuation_1], properties {inputVolume [1,1]}}
 {id=QualityLevel_0, children [Mixer_0,Mixer_0,null,null]}
 {id=Random_1, children [WavePlayer_0,WavePlayer_1], properties {weights [1,1]}}
 {id=WavePlayer_0, properties {soundWaveAssetPtr=&/Geode_Tear_Enter_01}}
 {id=WavePlayer_1, properties {soundWaveAssetPtr=&/Geode_Tear_Enter_02}}
]
properties {
 maxDistance=2097152
 priority=100
 soundClassObject=&Fort_Mission_Ingame_Important_PSM
 volumeMultiplier=0.55
}
