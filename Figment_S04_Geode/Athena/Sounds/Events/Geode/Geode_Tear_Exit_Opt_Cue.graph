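# Sound cue for the Geode tear exit sound, playing a single optimized wave.
# The wave feeds two parallel attenuations, close and distant, which a mixer blends at equal volume for distance-based layering.
# A quality-level root routes the mix only to the third and fourth slots, leaving the two lowest quality levels silent.
# It uses an important in-game mission sound class, high priority, reduced volume and a very large max distance.
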SoundCue /Figment_S04_Geode/Athena/Sounds/Events/Geode/Geode_Tear_Exit_Opt_Cue
root=QualityLevel_0
nodes [
 {id=Attenuation_0, children [WavePlayer_0], properties {attenuationSettings=&/Geode_Close_Att}}
 {id=Attenuation_1, children [WavePlayer_0], properties {attenuationSettings=&/Geode_Distant_Att}}
 {id=Mixer_0, children [Attenuation_0, Attenuation_1], properties {inputVolume [1,1]}}
 {id=QualityLevel_0, children [null,null,Mixer_0,Mixer_0]}
 {id=WavePlayer_0, properties {soundWaveAssetPtr=&/Geode_Tear_Exit_03_opt}}
]
properties {
 maxDistance=2097152
 priority=100
 soundClassObject=&Fort_Mission_Ingame_Important_PSM
 volumeMultiplier=0.55
}
