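# In-world sound cue for the Geode tear, layering audio by the CrackProgression parameter through a crossfade.
# A looping bed is always audible, randomly enveloped teleport oddities fade in early, and quiet distant thunder joins at higher progression.
# Both one-shot layers repeat after random delays, and quality-level nodes select optimized wave variants.
# Overrides disable distance attenuation, mute the cue when out of focus or occluded, and allow one instance.
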SoundCue /Figment_S04_Geode/Athena/Sounds/Events/Geode/Geode_Tear_InWorldLoop_Cue
root=ParamCrossFade_0
nodes [
 {id=Delay_0, children [Modulator_0], properties {delayMax=30, delayMin=10}}
 {id=Delay_1, children [QualityLevel_2], properties {delayMax=30, delayMin=10}}
 {
  id=Enveloper_0
  children [QualityLevel_1]
  properties {pitchCurve {editorCurveData {keys [{time=0}]}}, volumeCurve {editorCurveData {keys [<at 0 1>,<at 4 0>]}}}
 }
 {
  id=Enveloper_1
  children [QualityLevel_1]
  properties {pitchCurve {editorCurveData {keys [{time=0}, <at 4 0.6>]}}, volumeCurve {editorCurveData {keys [<at 0 0.1>,<at 2 1>,<at 3.6 0>]}}}
 }
 {
  id=Enveloper_2
  children [QualityLevel_1]
  properties {pitchCurve {editorCurveData {keys [<at 0 2>,<at 1.3509 1.4809>]}}, volumeCurve {editorCurveData {keys [<at 0 0.1>,<at 0.9321 1>,<at 2 0>]}}}
 }
 {id=Looping_0, children [Delay_0]}
 {id=Looping_1, children [Delay_1]}
 {id=Modulator_0, children [Random_0], properties {pitchMax=1, pitchMin=1, volumeMax=1, volumeMin=0.5}}
 {id=Modulator_1, children [WavePlayer_22], properties {pitchMax=1.1, pitchMin=0.8}}
 {
  id=ParamCrossFade_0
  children [Looping_1, Looping_0, QualityLevel_0]
  properties {
   paramName=CrackProgression
   crossFadeInput [
    {fadeInDistanceEnd=3, fadeInDistanceStart=2, fadeOutDistanceEnd=100, fadeOutDistanceStart=100, volume=0.1}
    {fadeInDistanceEnd=1, fadeInDistanceStart=0, fadeOutDistanceEnd=100, fadeOutDistanceStart=100, volume=1}
    {fadeInDistanceEnd=0, fadeInDistanceStart=0, fadeOutDistanceEnd=100, fadeOutDistanceStart=100, volume=1}
   ]
  }
 }
 {id=QualityLevel_0, children [WavePlayer_3,WavePlayer_3,WavePlayer_2,WavePlayer_2]}
 {id=QualityLevel_1, children [WavePlayer_0,WavePlayer_0,WavePlayer_1,WavePlayer_1]}
 {id=QualityLevel_2, children [WavePlayer_4,WavePlayer_4,Modulator_1,Modulator_1]}
 {id=Random_0, children [Enveloper_0,Enveloper_1,Enveloper_2], properties {weights [1,1,1]}}
 {id=WavePlayer_0, properties {soundWaveAssetPtr=/Figment_S04/Athena/Sounds/Events/Geode/Oddities/Geode_Oddities_Teleport_172}}
 {id=WavePlayer_1, properties {soundWaveAssetPtr=/Figment_S04/Athena/Sounds/Events/Geode/Oddities/Geode_Oddities_Teleport_17_opt2}}
 {id=WavePlayer_2, properties {bLooping=true, soundWaveAssetPtr=/Game/Athena/Sounds/Events/Geode/Oddities/Geode_Oddities_InWorldLoop_01_opt}}
 {id=WavePlayer_22, properties {soundWaveAssetPtr=/Game/Sounds/Fort_GamePlay_Sounds/Storms/Dist_Thunder_002_opt}}
 {id=WavePlayer_3, properties {bLooping=true, soundWaveAssetPtr=/Game/Athena/Sounds/Events/Geode/Oddities/Geode_Oddities_InWorldLoop_01}}
 {id=WavePlayer_4, properties {soundWaveAssetPtr=/Game/Sounds/Fort_GamePlay_Sounds/Storms/Dist_Thunder_001}}
]
properties {
 bOverrideAttenuation=true
 bOverrideConcurrency=true
 maxDistance=2097152
 soundClassObject=&Fort_SFX_InGame
 volumeMultiplier=2
 attenuationOverrides {
  bAttenuate=false
  bEnableFocusInterpolation=true
  bEnableListenerFocus=true
  bEnableOcclusion=true
  focusAttackInterpSpeed=0.1
  focusAzimuth=15
  focusReleaseInterpSpeed=0.5
  nonFocusAzimuth=30
  nonFocusVolumeAttenuation=0
  occlusionInterpolationTime=2
  occlusionTraceChannel=ECC_GameTraceChannel11
  occlusionVolumeAttenuation=0
  stereoSpread=30000
 }
 concurrencyOverrides {maxCount=1, resolutionRule=StopOldest}
}
