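# Sound cue for the wind freefall layer audio in the Figment S04 Geode event.
# A single wave player feeds the Geode_Tear_WindFreeFallLayer sound wave straight to the output, with no mixing or randomization.
# It plays at reduced volume through an important in-game mission sound class, with top priority and a very large max distance.
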SoundCue /Figment_S04_Geode/Athena/Sounds/Events/Geode/Geode_WindFreefallLayer_Cue
root=WavePlayer_1
nodes [{id=WavePlayer_1, properties {soundWaveAssetPtr=&/Geode_Tear_WindFreeFallLayer}}]
properties {
 maxDistance=2097152
 priority=100
 soundClassObject=&Fort_Mission_Ingame_Important_PSM
 volumeMultiplier=0.4
}
